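# Looping glider sound cue for the Galileo Ferry glider cosmetic, using player glider thrust attenuation and the cosmetic glider loop audio properties sheet.
# A looping wave player feeds a continuous modulator at the root.
# The TurnAmount parameter drives both pitch and volume, raising each up to 1.2x, with pitch responding over a narrower input range set just above 1.
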SoundCue /BRCosmetics/Sounds/Gliders/Galileo_Ferry/Glider_Galileo_Ferry_Loop_Cue
root=ModulatorContinuous_0
nodes [
 {
  id=ModulatorContinuous_0
  children [WavePlayer_0]
  properties {
   pitchModulationParams {default=0, maxInput=1.1, maxOutput=1.2, minInput=1, minOutput=1, parameterName=TurnAmount}
   volumeModulationParams {default=0, maxOutput=1.2, minOutput=1, parameterName=TurnAmount}
  }
 }
 {id=WavePlayer_0, properties {bLooping=true, soundWaveAssetPtr=&/Glider_Galileo_Ferry_Loop}}
]
properties {
 attenuationSettings=&Fort_Player_Glider_Thrust_Athena_Att
 audioPropertiesSheet=&APS_Cosmetic_Glider_Loop
 maxDistance=12500
 soundClassObject=&Fort_Gliders_Land
 volumeMultiplier=1
}
