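# Sound cue for the thrust audio of the Galileo Ferry glider cosmetic.
# A single wave player feeds the thrust wave straight to the output, with no randomization or mixing.
# The cue lowers volume and pitch and routes through a glider sound class and a cosmetic glider loop audio property sheet.
# Player glider thrust attenuation and a shared Athena glider thrust concurrency govern its spatial falloff and voice limiting.
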SoundCue /BRCosmetics/Sounds/Gliders/Galileo_Ferry/Glider_Galileo_Ferry_Thrust_Cue
root=WavePlayer_1
nodes [{id=WavePlayer_1, properties {soundWaveAssetPtr=&/Glider_Galileo_Ferry_Thrust}}]
properties {
 attenuationSettings=&Fort_Player_Glider_Thrust_Athena_Att
 audioPropertiesSheet=&APS_Cosmetic_Glider_Loop
 maxDistance=12500
 pitchMultiplier=0.9
 soundClassObject=&Fort_Gliders_Land
 volumeMultiplier=0.1
 concurrencySet [&Athena_Glider_Thrust_Conc]
}
